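# Physics asset for the CandyGirl female medium player head, giving simulated motion to dangling hair and earrings.
# Capsule bodies cover the head, six hair bones and both earrings.
# Limited swing and twist constraints tie the earrings and four hair roots to the head, while two of those strands chain into a second segment with very tight twist.
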PhysicsAsset /Game/Characters/Player/Female/Medium/Heads/F_MED_CandyGirl_Head/Mesh/F_MED_CandyGirl_Head_MESH_Physics
bodies {
 dyn_earring_l [{
   shapeType=Capsule
   length=2.932
   radius=1.9784
   shapeName=dyn_earring_l_capsule
   center=<0.6701 -0.591 -3.3216>
   rotation=<rot 9.108 -90 -6.8402>
  }]
 dyn_earring_r [{
   shapeType=Capsule
   length=2.932
   radius=1.9784
   shapeName=dyn_earring_r_capsule
   center=<-0.6773 0.591 3.3217>
   rotation=<rot 9.108 -90 -6.8402>
  }]
 dyn_hair_01 [{
   shapeType=Capsule
   length=9.8471
   radius=5.9824
   shapeName=dyn_hair_01_capsule
   center=<-0.2507 -8.7156 -1.6956>
   rotation=<rot 0 -173.5746 -55.5191>
  }]
 dyn_hair_01_a [{
   shapeType=Capsule
   length=5.581
   radius=3.6181
   shapeName=dyn_hair_01_a_capsule
   center=<-2.4601 1.3395 -5.369>
   rotation=<rot -8.3125 -90 37.6705>
  }]
 dyn_hair_02 [{
   shapeType=Capsule
   length=9.6421
   radius=6.5302
   shapeName=dyn_hair_02_capsule
   center=<0.2776 -8.1016 -1.7625>
   rotation=<rot 0 178.7679 -71.0142>
  }]
 dyn_hair_02_a [{
   shapeType=Capsule
   length=7.0311
   radius=4.6526
   shapeName=dyn_hair_02_a_capsule
   center=<-2.1312 -0.5994 -3.8943>
   rotation=<rot 8.1076 -90 40.8541>
  }]
 dyn_hair_03 [{
   shapeType=Capsule
   length=7.1289
   radius=4.3416
   shapeName=dyn_hair_03_capsule
   center=<-0.3133 -0.0607 -6.1878>
   rotation=<rot 2.491 -90 -21.9009>
  }]
 dyn_hair_04 [{
   shapeType=Capsule
   length=5.3418
   radius=4.9269
   shapeName=dyn_hair_04_capsule
   center=<0.5555 -1.2441 -5.8482>
   rotation=<rot -8.2515 -90 -2.0322>
  }]
 head [{shapeType=Capsule, length=11.889, radius=8.6253, center=<1.341 0.0571 -0.0001>, rotation=<rot 90 0 0>}]
}
constraints [
 {bone1Name=dyn_hair_01, bone2Name=head, swing1LimitDegrees=5, swing2LimitDegrees=15, twistLimitDegrees=20}
 {bone1Name=dyn_hair_01_a, bone2Name=dyn_hair_01, swing1LimitDegrees=20, swing2LimitDegrees=20, twistLimitDegrees=3}
 {bone1Name=dyn_hair_02, bone2Name=head, swing1LimitDegrees=5, swing2LimitDegrees=15, twistLimitDegrees=20}
 {bone1Name=dyn_hair_02_a, bone2Name=dyn_hair_02, swing1LimitDegrees=20, swing2LimitDegrees=20, twistLimitDegrees=3}
 {bone1Name=dyn_earring_l, bone2Name=head, swing1LimitDegrees=20, swing2LimitDegrees=20, twistLimitDegrees=20}
 {bone1Name=dyn_earring_r, bone2Name=head, swing1LimitDegrees=20, swing2LimitDegrees=20, twistLimitDegrees=20}
 {bone1Name=dyn_hair_03, bone2Name=head, swing1LimitDegrees=15, swing2LimitDegrees=15, twistLimitDegrees=15}
 {bone1Name=dyn_hair_04, bone2Name=head, swing1LimitDegrees=15, swing2LimitDegrees=15, twistLimitDegrees=15}
]
