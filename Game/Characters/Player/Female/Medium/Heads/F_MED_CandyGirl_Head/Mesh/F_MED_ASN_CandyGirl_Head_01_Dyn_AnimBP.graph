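# Post-style animation blueprint for the CandyGirl female head that adds secondary motion to its hair and earrings on top of the input pose.
# bRigidBodyEnabled selects either a RigidBody simulation in head-bone space or a fallback chain of per-bone AnimDynamics nodes.
# A playlist flag for custom character parts, read at initialization, overrides both by scaling the head bone to near zero.
# All branches share one cached input pose.
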AnimBlueprint /Game/Characters/Player/Female/Medium/Heads/F_MED_CandyGirl_Head/Mesh/F_MED_ASN_CandyGirl_Head_01_Dyn_AnimBP
target_skeleton=/Game/Characters/Player/Common/Fortnite_Base_Head/Fortnite_Base_Head_Export_Skeleton
graphs [
 AnimGraph {nodes [
   n1 = RigidBody(ComponentPose=Local To Component(LocalPose=Use cached pose 'PreScale Cache'())) {
    bOverrideWorldGravity=true
    overridePhysicsAsset=&/F_MED_CandyGirl_Head_MESH_Physics
    simulationSpace=BaseBoneSpace
    alphaBoolBlend {blendOption=Linear}
    baseBoneRef {boneName=head}
    overrideWorldGravity {z=-980}
    simSpaceSettings {
     dampingAlpha=0.9
     maxAngularAcceleration=20
     maxAngularVelocity=4
     maxLinearAcceleration=2000
     maxLinearVelocity=400
     velocityScaleZ=0.5
     worldAlpha=1
     externalLinearDragV=<0.1 0.2 0.3>
    }
   }
   n2 = Component To Local(ComponentPose=@n1)
   n3 = Anim Dynamics - Bone: dyn_hair_01(ComponentPose=Local To Component(LocalPose=Use cached pose 'PreScale Cache'()), Alpha=1) {
    alpha=0.9
    angularDampingOverride=0.9
    angularSpringConstant=1500
    bAngularSpring=true
    bGravityOverrideInSimSpace=true
    bOverrideAngularDamping=true
    numSolverIterationsPostUpdate=4
    numSolverIterationsPreUpdate=16
    simulationSpace=Actor
    alphaBoolBlend {blendOption=Linear}
    boundBone {boneName=dyn_hair_01}
    physicsBodyDefinitions [{
      sphereCollisionRadius=0
      boundBone {boneName=dyn_hair_01}
      boxExtents=<15 5 5>
      constraintSetup {angularTargetAxis=AxisZ, twistAxis=AxisY, angularLimitsMax=<15 0 15>, angularLimitsMin=<-15 0 -15>, angularTarget=<0 0 1>}
      localJointOffset=<0 0 0>
     }]
   }
   n4 = Anim Dynamics - Bone: dyn_hair_02(ComponentPose=@n3, Alpha=1) {
    alpha=0.9
    angularDampingOverride=0.9
    angularSpringConstant=1500
    bAngularSpring=true
    bGravityOverrideInSimSpace=true
    bOverrideAngularDamping=true
    numSolverIterationsPostUpdate=4
    numSolverIterationsPreUpdate=16
    simulationSpace=Actor
    alphaBoolBlend {blendOption=Linear}
    boundBone {boneName=dyn_hair_02}
    physicsBodyDefinitions [{
      sphereCollisionRadius=0
      boundBone {boneName=dyn_hair_02}
      boxExtents=<15 5 5>
      constraintSetup {angularTargetAxis=AxisZ, twistAxis=AxisY, angularLimitsMax=<15 0 15>, angularLimitsMin=<-15 0 -15>, angularTarget=<0 0 1>}
      localJointOffset=<0 0 0>
     }]
   }
   n5 = Anim Dynamics - Bone: dyn_hair_01_a(ComponentPose=@n4, Alpha=1, ExternalForce="0,0,0") {
    alpha=0.9
    angularDampingOverride=0.9
    angularSpringConstant=1500
    bAngularSpring=true
    bGravityOverrideInSimSpace=true
    bOverrideAngularDamping=true
    numSolverIterationsPostUpdate=4
    numSolverIterationsPreUpdate=16
    simulationSpace=Actor
    alphaBoolBlend {blendOption=Linear}
    boundBone {boneName=dyn_hair_01_a}
    physicsBodyDefinitions [{
      sphereCollisionRadius=0
      boundBone {boneName=dyn_hair_01_a}
      boxExtents=<15 5 5>
      constraintSetup {angularTargetAxis=AxisZ, twistAxis=AxisY, angularLimitsMax=<15 15 15>, angularLimitsMin=<-15 -15 -15>, angularTarget=<0 0 1>}
      localJointOffset=<0 0 0>
     }]
   }
   n6 = Anim Dynamics - Bone: dyn_hair_02_a(ComponentPose=@n5, Alpha=1, ExternalForce="0,0,0") {
    alpha=0.9
    angularDampingOverride=0.9
    angularSpringConstant=1500
    bAngularSpring=true
    bGravityOverrideInSimSpace=true
    bOverrideAngularDamping=true
    numSolverIterationsPostUpdate=4
    numSolverIterationsPreUpdate=16
    simulationSpace=Actor
    alphaBoolBlend {blendOption=Linear}
    boundBone {boneName=dyn_hair_02_a}
    physicsBodyDefinitions [{
      sphereCollisionRadius=0
      boundBone {boneName=dyn_hair_02_a}
      boxExtents=<15 5 5>
      constraintSetup {angularTargetAxis=AxisZ, twistAxis=AxisY, angularLimitsMax=<15 15 15>, angularLimitsMin=<-15 -15 -15>, angularTarget=<0 0 1>}
      localJointOffset=<0 0 0>
     }]
   }
   n7 = Anim Dynamics - Bone: dyn_earring_l(ComponentPose=@n6, Alpha=1, ExternalForce="0,0,0") {
    alpha=0.9
    angularDampingOverride=0.94
    angularSpringConstant=500
    bGravityOverrideInSimSpace=true
    bOverrideAngularDamping=true
    numSolverIterationsPostUpdate=4
    numSolverIterationsPreUpdate=16
    alphaBoolBlend {blendOption=Linear}
    boundBone {boneName=dyn_earring_l}
    physicsBodyDefinitions [{
      sphereCollisionRadius=0
      boundBone {boneName=dyn_earring_l}
      boxExtents=<15 15 5>
      constraintSetup {angularTargetAxis=AxisZ, twistAxis=AxisY, angularLimitsMax=<180 30 80>, angularLimitsMin=<-180 -30 -80>, angularTarget=<0 0 1>}
      localJointOffset=<0 0 -1>
     }]
   }
   n8 = Anim Dynamics - Bone: dyn_earring_r(ComponentPose=@n7, Alpha=1, ExternalForce="0,0,0") {
    alpha=0.9
    angularDampingOverride=0.95
    angularSpringConstant=500
    bGravityOverrideInSimSpace=true
    bOverrideAngularDamping=true
    numSolverIterationsPostUpdate=4
    numSolverIterationsPreUpdate=16
    alphaBoolBlend {blendOption=Linear}
    boundBone {boneName=dyn_earring_r}
    physicsBodyDefinitions [{
      sphereCollisionRadius=0
      boundBone {boneName=dyn_earring_r}
      boxExtents=<15 15 5>
      constraintSetup {angularTargetAxis=AxisZ, twistAxis=AxisY, angularLimitsMax=<180 30 80>, angularLimitsMin=<-180 -30 -80>, angularTarget=<0 0 1>}
      localJointOffset=<0 0 1>
     }]
   }
   n9 = Anim Dynamics - Bone: dyn_hair_03(ComponentPose=@n8, Alpha=0.4, ExternalForce="0,0,0") {
    alpha=0.9
    angularSpringConstant=1500
    bGravityOverrideInSimSpace=true
    bOverrideAngularDamping=true
    numSolverIterationsPostUpdate=4
    numSolverIterationsPreUpdate=16
    alphaBoolBlend {blendOption=Linear}
    boundBone {boneName=dyn_hair_03}
    physicsBodyDefinitions [{
      sphereCollisionRadius=0
      boundBone {boneName=dyn_hair_03}
      boxExtents=<15 15 5>
      constraintSetup {angularTargetAxis=AxisZ, twistAxis=AxisY, angularLimitsMax=<20 20 0>, angularLimitsMin=<-20 -20 0>, angularTarget=<0 0 1>}
      localJointOffset=<0 0 -1>
     }]
   }
   n10 = Anim Dynamics - Bone: dyn_hair_04(ComponentPose=@n9, Alpha=0.4, ExternalForce="0,0,0") {
    alpha=0.9
    angularSpringConstant=1500
    bGravityOverrideInSimSpace=true
    bOverrideAngularDamping=true
    numSolverIterationsPostUpdate=4
    numSolverIterationsPreUpdate=16
    alphaBoolBlend {blendOption=Linear}
    boundBone {boneName=dyn_hair_04}
    physicsBodyDefinitions [{
      sphereCollisionRadius=0
      boundBone {boneName=dyn_hair_04}
      boxExtents=<15 15 5>
      constraintSetup {angularTargetAxis=AxisZ, twistAxis=AxisY, angularLimitsMax=<20 20 0>, angularLimitsMin=<-20 -20 0>, angularTarget=<0 0 1>}
      localJointOffset=<0 0 -1>
     }]
   }
   n11 = Component To Local(ComponentPose=@n10)
   n12 = Get bRigidBodyEnabled()
   n13 = "Transform (Modify) Bone - Bone: head"!(ComponentPose=Local To Component(LocalPose=Use cached pose 'PreScale Cache'()), Alpha=1) -> (Pose) {scaleMode=BMM_Replace, alphaBoolBlend {blendOption=Linear}, boneToModify {boneName=head}, scale=<0.01 0.01 0.01>}
   Output Pose(Result=Blend Poses by bool(
     BlendPose_0=Component To Local(ComponentPose=@n13)
     BlendPose_1=Blend Poses by bool(BlendPose_0=@n2, BlendPose_1=@n11, bActiveValue=@n12) {blendType=Linear}
     BlendTime_0=0
     BlendTime_1=0
     bActiveValue=Get GetPlayerListUsesCustomCharacterParts() -> (GetPlayerListUsesCustomCharacterParts:bool)
    ) {blendType=Linear}) {layerGroup=None}
   AnimGraphNode_SubInput_0 = Input Pose()
   Save cached pose 'PreScale Cache'(Pose=@AnimGraphNode_SubInput_0) {cacheName=PreScale Cache}
  ]}
 EventGraph {nodes [
   Event BlueprintInitializeAnimation()
   > Set GetPlayerListUsesCustomCharacterParts(GetPlayerListUsesCustomCharacterParts:bool=GetPlaylistUsesCustomCharacterParts!() -> (ReturnValue:bool)) -> (Output_Get:bool)
   TryGetPawnOwner()
   Event BlueprintUpdateAnimation()
  ]}
]
variables [GetPlayerListUsesCustomCharacterParts {type=boolean}]
